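# Niagara system template for a 2D grid shallow water liquid simulation, in a variant named Drop.
# A single GPU emitter, ShallowWaterEmitter, runs the simulation and draws through a mesh renderer.
# User parameters set world grid size, max-axis resolution and bottom depth.
# Rigid mesh collision query and landscape data interfaces are exposed as inputs for scene colliders and terrain.
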NiagaraSystem /NiagaraFluids/Templates/Liquid/2D/Systems/ShallowWater/Grid2D_SW_Drop
dependencies=null
summary {
 systemName=Grid2D_SW_Drop
 emitters [{emitterName=ShallowWaterEmitter, simTarget=GPUComputeSim, rendererClasses [&NiagaraMeshRendererProperties]}]
 userVariables [
  User.BottomDepth {type=NiagaraFloat, defaultValue=-50, description=""}
  User.Collide_Meshes {
   type=NiagaraDataInterfaceRigidMeshCollisionQuery
   description=""
   defaultValue {
    struct=NiagaraExt_VariableValue_DataInterface
    value {
     dataInterface=&self:NiagaraDataInterfaceRigidMeshCollisionQuery_0
     dataInterfaceClass=/Script/Niagara.NiagaraDataInterfaceRigidMeshCollisionQuery
    }
   }
  }
  User.Landscape {
   type=NiagaraDataInterfaceLandscape
   description=""
   defaultValue {
    struct=NiagaraExt_VariableValue_DataInterface
    value {
     dataInterface=&self:NiagaraDataInterfaceLandscape_0
     dataInterfaceClass=&NiagaraDataInterfaceLandscape
    }
   }
  }
  User.ResolutionMaxAxis {type=NiagaraInt32, defaultValue=64}
  User.WorldGridSize {type=Vector2f, description="", defaultValue {struct=Vector2f, value=<1000 1000>}}
 ]
}
