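# Niagara Fluids template system for 2D shallow water with particle collisions, running on the GPU.
# A ShallowWaterEmitter is paired with one red mesh collision particle that swings along X around the system position on a cosine, with radius and velocity computed for collision.
# A disabled secondary sprite emitter reads the water and velocity grids through Grid2D readers and uses scratch modules with emission-position and counter grids.
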NiagaraSystem /NiagaraFluids/Templates/Liquid/2D/Systems/ShallowWater/Grid2D_SW_ParticleCollisions
dependencies=null
emitters {
 ShallowWaterCollisionParticles {
  simTarget=GPUComputeSim
  emitterUpdateScript [
   &EmitterState(
    Life Cycle Mode=::NewEnumerator1
    Loop Behavior=::NewEnumerator1
    Loop Duration Mode=::NewEnumerator1
    Scalability Mode=::NewEnumerator1
    UseLoopDelay=false
   )
   &SpawnBurst_Instantaneous(Spawn Count=1, Use Spawn Probability=false)
  ]
  particleSpawnScript [
   /SetVariables(Particles.ParticleRadius=100) {isSetParameters=true}
   &Color(Color=<rgba 1 0 0 1>)
   /SetVariables(
    Particles.Position=<0 0 0>
    Particles.Radius:NiagaraFloat=$Particles.ParticleRadius
    Particles.RadiusSquared:NiagaraFloat=&Multiply_Float(A=$Particles.ParticleRadius, B=$Particles.ParticleRadius)
    Particles.Scale=&Divide_Vector(
     A:Vector3f=&VectorFromFloat(Value=$Particles.ParticleRadius)
     B:Vector3f=&VectorFromFloat(Value=50)
    )
   ) {isSetParameters=true}
  ]
  particleUpdateScript [
   /SetVariables(Particles.Position=&Add_Vector(
     A=&ConvertPositionToVector(Input Position=$Engine.Owner.Position, Convert To Absolute Large World Vector=false)
     B=&MakeVector(X=&Cosine(Normalized Angle=$Emitter.Age, Period=8, Scale=800, Bias=0), Y=0, Z=0)
    )) {isSetParameters=true}
   &CalculateAccurateVelocity(Calculate Accurate Velocity On First Frame=true)
  ]
  rendererClasses [&NiagaraMeshRendererProperties]
  renderers [{rendererClass=&NiagaraMeshRendererProperties, rendererIndex=0}]
 }
 ShallowWater_Secondary_Emitter {
  bEnabled=false
  simTarget=GPUComputeSim
  emitterSpawnScript [
   /SetVariables(
    Emitter.VelocityReader=DataInterface() {propertyValues {
      ClearBeforeNonIterationStage=true
      DIName=VelocityGrid
      EmitterName=ShallowWaterEmitter
      NumAttributes=0
      NumCellsMaxAxis=3
      NumCellsX=3
      NumCellsY=3
      OverrideBufferFormat=Float
      PreviewAttribute=None
      SetGridFromMaxAxis=false
      bOverrideFormat=false
      bPreviewGrid=false
      RenderTargetUserParameter {parameter {name=None, typeDefHandle {registeredTypeIndex=4}, varData []}}
      WorldBBoxSize=<100 100>
     }}
    Emitter.WaterReader=DataInterface() {propertyValues {
      ClearBeforeNonIterationStage=true
      DIName=WaterGrid
      EmitterName=ShallowWaterEmitter
      NumAttributes=0
      NumCellsMaxAxis=3
      NumCellsX=3
      NumCellsY=3
      OverrideBufferFormat=Float
      PreviewAttribute=None
      SetGridFromMaxAxis=false
      bOverrideFormat=false
      bPreviewGrid=false
      RenderTargetUserParameter {parameter {name=None, typeDefHandle {registeredTypeIndex=4}, varData []}}
      WorldBBoxSize=<100 100>
     }}
   ) {isSetParameters=true}
   /SetVariables(Emitter.MaxSecondaryParticlesPerFrame=16000) {isSetParameters=true}
   /SetVariables(Emitter.EmissionPositions=DataInterface() {propertyValues {
      ClearBeforeNonIterationStage=true
      NumAttributes=2
      NumCellsMaxAxis=3
      NumCellsX=1
      NumCellsY=1
      OverrideBufferFormat=Float
      PreviewAttribute=None
      SetGridFromMaxAxis=false
      bOverrideFormat=false
      bPreviewGrid=false
      RenderTargetUserParameter {parameter {name=None, typeDefHandle {registeredTypeIndex=4}, varData []}}
      WorldBBoxSize=<100 100>
     }}) {isSetParameters=true}
   /NiagaraFluids/Modules/Grid2D/Grid2D_SetResolution(
    Grid:NiagaraDataInterfaceGrid2DCollection=$Emitter.EmissionPositions
    NumCellsX:NiagaraInt32=$Emitter.MaxSecondaryParticlesPerFrame
    NumCellsY=1
   )
   /SetVariables(Emitter.EmissionCounter=DataInterface() {propertyValues {
      CellSize=1
      ClearBeforeNonIterationStage=true
      MaxNeighborsPerCell=1
      NumCellsMaxAxis=10
      SetResolutionMethod=Independent
      NumCells=<2 2 2>
      WorldBBoxSize=<100 100 100>
     }}) {isSetParameters=true}
   &self:ShallowWater_Secondary_Emitter.NiagaraScratchPadContainer_0.ScratchModule_01_5() {module=ScratchModule_01}
  ]
  emitterUpdateScript [
   &EmitterState()
   &self:ShallowWater_Secondary_Emitter.NiagaraScratchPadContainer_0.ScratchModule_06_1(Extents:Vector2f=$User.WorldGridSize, Read:NiagaraDataInterfaceGrid2DCollectionReader=$Emitter.WaterReader) {module=ScratchModule_06}
   /NiagaraFluids/Modules/Grid2D/Grid2D_CreateUnitToWorldTransform(Offset:NiagaraPosition=$Engine.Owner.Position, Rotation:Quat4f=$Engine.Owner.Rotation, WorldGridExtents:Vector2f=$User.WorldGridSize)
   /SetVariables(
    Emitter.LocalToWorld:NiagaraMatrix=$Emitter.Grid2D_CreateUnitToWorldTransform.LocalToWorld
    Emitter.UnitToWorld:NiagaraMatrix=$Emitter.Grid2D_CreateUnitToWorldTransform.UnitToWorld
    Emitter.WorldToLocal:NiagaraMatrix=$Emitter.Grid2D_CreateUnitToWorldTransform.WorldToLocal
    Emitter.WorldToUnit:NiagaraMatrix=$Emitter.Grid2D_CreateUnitToWorldTransform.WorldToUnit
   ) {isSetParameters=true}
   /SetVariables(Emitter.DeltaTime:NiagaraFloat=$Engine.DeltaTime) {isSetParameters=true}
   &SpawnPerFrame(Spawn Count:NiagaraInt32=$Emitter.MaxSecondaryParticlesPerFrame, Use Spawn Probability=false)
  ]
  particleSpawnScript [&self:ShallowWater_Secondary_Emitter.NiagaraScratchPadContainer_0.ScratchModule_0(
    Lifetime:NiagaraFloat=&RandomRangeFloat(Minimum=2, Maximum=3)
    SpriteSize:Vector2f=&Vector2DFromFloat(Value=&FloatFromCurve(
      FloatCurve=DataInterface() {propertyValues {Curve {keys [<at 0 5 RCIM_Cubic>,<at 0.9 8>]}}}
      CurveIndex=&RandomRangeFloat(Minimum=0, Maximum=1)
     ))
   ) {module=ScratchModule_4}]
  rendererClasses [&NiagaraSpriteRendererProperties]
  renderers [{rendererClass=&NiagaraSpriteRendererProperties}]
 }
}
summary {
 systemName=Grid2D_SW_ParticleCollisions
 emitters [
  {emitterName=ShallowWaterEmitter, simTarget=GPUComputeSim, rendererClasses [&NiagaraMeshRendererProperties]}
  {emitterName=ShallowWaterCollisionParticles, simTarget=GPUComputeSim, rendererClasses [&NiagaraMeshRendererProperties]}
  {
   bEnabled=false
   emitterName=ShallowWater_Secondary_Emitter
   simTarget=GPUComputeSim
   rendererClasses [&NiagaraSpriteRendererProperties]
  }
 ]
 userVariables [
  User.BottomDepth {type=NiagaraFloat, defaultValue=-50, description=""}
  User.Collide_Meshes {
   type=NiagaraDataInterfaceRigidMeshCollisionQuery
   description=""
   defaultValue {
    struct=NiagaraExt_VariableValue_DataInterface
    value {
     dataInterface=&self:NiagaraDataInterfaceRigidMeshCollisionQuery_0
     dataInterfaceClass=/Script/Niagara.NiagaraDataInterfaceRigidMeshCollisionQuery
    }
   }
  }
  User.Landscape {
   type=NiagaraDataInterfaceLandscape
   description=""
   defaultValue {
    struct=NiagaraExt_VariableValue_DataInterface
    value {
     dataInterface=&self:NiagaraDataInterfaceLandscape_0
     dataInterfaceClass=&NiagaraDataInterfaceLandscape
    }
   }
  }
  User.ResolutionMaxAxis {type=NiagaraInt32, defaultValue=128}
  User.WorldGridSize {type=Vector2f, description="", defaultValue {struct=Vector2f, value=<2000 2000>}}
 ]
}
